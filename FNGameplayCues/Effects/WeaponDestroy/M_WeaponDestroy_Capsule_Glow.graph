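# Additive unlit Niagara sprite material for a capsule-shaped glow in the weapon destroy gameplay cue.
# Sprite_Capsule stretches sprites along velocity, with a previous-frame offset switch compensating for particle movement.
# A radial mask eroded by panning, per-particle-randomized noise is faded by depth and camera distance, then drives dark-to-bright tinting by particle color with HDR remapping.
# A particle-direction versus camera-direction term is computed for adjusting end-on views.
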Material /FNGameplayCues/Effects/WeaponDestroy/M_WeaponDestroy_Capsule_Glow
graph {
 properties {
  bCastRayTracedShadows=false
  bUsedWithNiagaraSprites=true
  blendMode=BLEND_Additive
  pixelDepthOffsetMode=PDOM_Legacy
  shadingModel=MSM_Unlit
  usedShadingModels=Unlit
 }
 nodes [
  Constant() {r=1}
  n1 = ParticleColor()
  n2 = ParticleSize()
  n3 = StaticBool()
  n4 = MaterialFunctionCall!(
   Particle Position
   Normalized Particle Direction
   Capsule Diameter
   Capsule Pivot Alignment
   Scale Along Velocity Vector
   ---------Scale Over Velocity Parameters---------
   Particle Speed
   Minimum Speed
   Max Speed
   Minimum Scale Factor=MaterialFunctionCall(Float2=ParticleSize()) {materialFunction=&BreakOutFloat2Components}.G
   Maximum Scale Factor
   ---------- Advanced -----------
   Cap Size Multiplier=LinearInterpolate(B=ScalarParameter() {defaultValue=1.5, parameterName=Cap Size Multiplier, sliderMax=1}, Alpha=NamedRerouteUsage()) {constA=0.9, constB=0.95}
   Use Camera Offset=@n3
   Use Advanced Features=@n3
   World Position to Negate
  ) {materialFunction=/Engine/Functions/Engine_MaterialFunctions02/WorldPositionOffset/Sprite_Capsule}
  n5 = SmoothStep(Value=Distance(A=@n4.Texture Coordinates, B=Constant() {r=0.5})) {constMax=0.5}
  n6 = Add(A=TextureCoordinate(), B=AppendVector(A=ParticleRandom(), B=Constant()))
  n7 = MaterialFunctionCall(Float2=ParticleSize()) {materialFunction=&BreakOutFloat2Components}
  n8 = Divide(A=@n7.G, B=@n7.G)
  n9 = ScalarParameter() {defaultValue=0.25, parameterName=Noise Frequency}
  n10 = AppendVector(A=LinearInterpolate(B=Multiply(A=@n8, B=@n9) {constB=0.2}, Alpha=NamedRerouteUsage()) {constA=0.001, constB=0.2}, B=Constant() {r=0.001})
  n11 = Time()
  n12 = AppendVector(A=Constant(), B=ScalarParameter() {defaultValue=1, parameterName=UV Panning})
  n13 = TextureSample(UVs=Add(A=Multiply(A=@n6, B=@n10), B=Multiply(A=@n11, B=@n12))) {texture=/Game/Global/Textures/Default/Noise/T_SoftCells_Negative}
  n14 = Multiply(A=@n1.A, B=Subtract(A=OneMinus(None=@n5), B=LinearInterpolate(Alpha=@n13.G) {constB=0.33})) {constA=1}
  n15 = DepthFade(
   Opacity=@n14
   FadeDistance=Multiply(A=MaterialFunctionCall(Float2=@n2) {materialFunction=&BreakOutFloat2Components}.R) {constB=0.5}
  )
  n16 = ScalarParameter() {defaultValue=100, parameterName=Camera depth Fade length}
  n17 = ScalarParameter() {defaultValue=10, parameterName=Camera depth Fade Offset}
  n18 = Power(Base=Multiply(
    A=ShadingPathSwitch(Default=@n15, Forward=@n14, Mobile=@n14)
    B=MaterialFunctionCall(Fade Length=@n16, Fade Offset=@n17) {materialFunction=&CameraDepthFade}.Result
   )) {constExponent=1}
  n19 = DynamicParameter() {defaultValue=<rgba 4 20 1 1>, paramNames [Power,DepthFade,HDRMult,BottomFade]}
  n20 = Multiply(
   A=@n1.RGB
   B=LinearInterpolate(
    A=ScalarParameter() {defaultValue=0.05, parameterName=Glow Dark Value}
    B=ScalarParameter() {defaultValue=1, parameterName=Glow Bright Value}
    Alpha=Power(Base=@n18, Exp=@n19.Power)
   )
  )
  n21 = Clamp(None=@n20)
  MaterialFunctionCall(Emissive=ShadingPathSwitch(
    Default=MaterialFunctionCall(LDR Input=@n21, HDR Tint=@n20, HDR Intensity Multiplier=@n19.HDRMult) {materialFunction=&DeriveHDRfromLDR}.RemappedRange
    Mobile=@n20
   )) {materialFunction=&MF_FX_EmissiveAdjustments}
  n22 = Abs(None=DotProduct(
    A=ParticleDirection()
    B=MaterialFunctionCall() {materialFunction=&CameraDirectionVector}.Result
   ))
  n23 = OneMinus(None=Power(Base=Saturate(None=@n22)) {constExponent=5})
  NamedRerouteDeclaration(None=@n23) {name=Direction To Camera, nodeColor=<rgba 1 0 0.2109 1>}
  n24 = Multiply(A=Multiply(A=ParticleDirection(), B=ParticleSpeed()), B=DeltaTime())
  PreviousFrameSwitch(Current Frame=@n4.World Position Offset, Previous Frame=Subtract(A=@n4.World Position Offset, B=Multiply(A=@n24)))
  SmoothStep(Value=@n23) {constMax=0.66, constMin=0.05}
 ]
}
